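; BP_GrumpyDamageComponent is a Grumpy prototype component that relays integer damage amounts for nine numbered props.
; Custom events DamageProp0 through DamageProp8 each broadcast the OnDamageAdded dispatcher with the damage amount and that prop's index.
; It applies no damage itself, leaving the response to the dispatcher's listeners; BeginPlay and Tick are empty.
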(event EventBeginPlay)

(event EventTick
  (params (DeltaSeconds :type Float%20%28single-precision%29)))

(event Custom|DamageProp0
  (params (DamageApplied :type Integer))
  (Default|CallOnDamageAdded DamageApplied))

(event Custom|DamageProp1
  (params (DamageApplied :type Integer))
  (Default|CallOnDamageAdded DamageApplied 1))

(event Custom|DamageProp2
  (params (DamageApplied :type Integer))
  (Default|CallOnDamageAdded DamageApplied 2))

(event Custom|DamageProp3
  (params (DamageApplied :type Integer))
  (Default|CallOnDamageAdded DamageApplied 3))

(event Custom|DamageProp4
  (params (DamageApplied :type Integer))
  (Default|CallOnDamageAdded DamageApplied 4))

(event Custom|DamageProp5
  (params (DamageApplied :type Integer))
  (Default|CallOnDamageAdded DamageApplied 5))

(event Custom|DamageProp6
  (params (DamageApplied :type Integer))
  (Default|CallOnDamageAdded DamageApplied 6))

(event Custom|DamageProp7
  (params (DamageApplied :type Integer))
  (Default|CallOnDamageAdded DamageApplied 7))

(event Custom|DamageProp8
  (params (DamageApplied :type Integer))
  (Default|CallOnDamageAdded DamageApplied 8))
