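; Damageable Grumpy cube panel that shows red, green or blue visuals and never dies from damage.
; Server damage is broadcast through OnDamageReceived, while EdgeServerApplyDamage sets its health from a supplied 64-bit total.
; A forward-only replicated completion state drives client setup between active and completed looks, toggling field, beam, audio and health bar.
; Timelines animate color changes and the success sequence, which fades the floor and drops the eye sphere before marking completion.
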(event Custom|OnBuildingHealthChanged_Event
  (Building|RequestBuildingFullUpdate))

(event EventBeginPlay
  (Class|FortAsyncActionCurrentPlaylistReady|PlaylistDataReadyAsync
    (:OnReady
      (Networking|SwitchHasAuthority
        (:Authority
          (CallFunction|UpdateCompleteState 0)
          (GameplayEffects|ApplyGameplayEffectToSelf :self (Ability|GetAbilitySystemComponent self) :GameplayEffectClass "/Figment_S08/Athena/Prototype/Blueprints/Iceberg/GE_MaxHealthSet.GE_MaxHealthSet_C" :Level 0.0))
        (:Remote
          (Health|AssignOnBuildingHealthChanged (EventDispatchers|CreateEvent :!event Custom|OnBuildingHealthChanged_Event)))))))

(event Custom|ActiveSetup
  (bind _gridambient (Variables|Default|GetGridAmbient))
  (CallFunction|UpdateClientVisuals)
  (if true
    (Components|Activation|Activate (Variables|Default|GetNS_Grumpy_Beam) false)
    (Development|PrintString "ACTIVATE" true true "(R=1.000000,G=0.000000,B=0.790744,A=1.000000)")
    (Audio|Components|Audio|FadeIn (Variables|Default|GetAudioAmbient) 1.0 1.0 0.0 "Linear" :!class /Script/Engine.AudioComponent)
    (if (Variables|Default|GetCenterOfTheGrid)
      (Utilities|IsValid _gridambient
        (:Is%20Valid
          (Audio|Components|Audio|FadeIn _gridambient 1.0 1.0 0.0 "Linear" :!class /Script/Engine.AudioComponent))
        (:Is%20Not%20Valid
          (bind _returnvalue (Audio|SpawnSoundAttached "/Figment_S08/Athena/Sounds/Events/Dopey/Grumpy/Grumpy_Grid_Amb_Cue.Grumpy_Grid_Amb_Cue" (Variables|Default|GetAudioAmbient)))
          (bind _output_get (Variables|Default|SetGridAmbient _returnvalue))
          (Audio|Components|Audio|FadeIn _output_get 1.0 1.0 0.0 "Linear" :!class /Script/Engine.AudioComponent))))))

(event EventOnDamageServer
  (params (Damage :type Float%20%28single-precision%29) (DamageTags :type Gameplay%20Tag%20Container%20Structure%20%28by%20ref%29) (Momentum :type Vector :default "0, 0, 0") (HitInfo :type Hit%20Result%20Structure%20%28by%20ref%29) (InstigatedBy :type Controller%20Object%20Reference) (DamageCauser :type Actor%20Object%20Reference) (EffectContext :type Gameplay%20Effect%20Context%20Handle%20Structure))
  (|CallParentFunction Damage DamageTags Momentum HitInfo InstigatedBy DamageCauser EffectContext)
  (if (Utilities|Equal(Object) DamageCauser self)
    _
    (elif (> Damage 0.0)
      (Default|CallOnDamageReceived (Math|Float|Truncate Damage) self))))

(event EventOnDamagePlayEffects
  (params (Damage :type Float%20%28single-precision%29) (DamageTags :type Gameplay%20Tag%20Container%20Structure%20%28by%20ref%29) (Momentum :type Vector :default "0, 0, 0") (HitInfo_bBlockingHit :type Boolean) (HitInfo_bInitialOverlap :type Boolean) (HitInfo_Time :type Float%20%28single-precision%29) (HitInfo_Distance :type Float%20%28single-precision%29) (HitInfo_Location :type Vector :default "0, 0, 0") (HitInfo_ImpactPoint :type Vector :default "0, 0, 0") (HitInfo_Normal :type Vector :default "0, 0, 0") (HitInfo_ImpactNormal :type Vector :default "0, 0, 0") (HitInfo_PhysMat :type Physical%20Material%20Object%20Reference) (HitInfo_HitActor :type Actor%20Object%20Reference) (HitInfo_HitComponent :type Primitive%20Component%20Object%20Reference) (HitInfo_HitBoneName :type Name :default "None") (HitInfo_BoneName :type Name :default "None") (HitInfo_HitItem :type Integer) (HitInfo_ElementIndex :type Integer) (HitInfo_FaceIndex :type Integer) (HitInfo_TraceStart :type Vector :default "0, 0, 0") (HitInfo_TraceEnd :type Vector :default "0, 0, 0") (InstigatedBy :type Fort%20Pawn%20Object%20Reference) (DamageCauser :type Actor%20Object%20Reference) (EffectContext :type Gameplay%20Effect%20Context%20Handle%20Structure))
  (bind _fieldfx (Variables|Default|GetFieldFX))
  ;[1 1 1] Event on Play Damage Effects
  (if (Math|Byte|Equal(Byte) (Variables|Default|GetRepCompleteState) 0)
    (Variables|Default|SetHit%20Impact%20Location HitInfo_ImpactPoint)
    (GameplayCue|ExecuteGameplayCue%28Burst%29Local self (Variables|Default|GetCueTag_SleepyImpact) (Ability|GameplayCue|MakeGameplayCueParameters :NormalizedMagnitude 0.0 :RawMagnitude (/ Damage (Variables|Default|GetMagnitudeDivisor)) :Location HitInfo_ImpactPoint :Normal HitInfo_ImpactNormal :Instigator (Variables|Actor|GetInstigator) :EffectCauser self :GameplayEffectLevel 1 :AbilityLevel 1 :bReplicateLocationWhenUsingMinimalRepProxy false))
    (Rendering|Material|SetVectorParameterValueonMaterials _fieldfx "HitWave-Loc" (Variables|Default|GetHit%20Impact%20Location))
    (Rendering|Material|SetScalarParameterValueonMaterials _fieldfx "GameTime" (Utilities|Time|GetGameTimeinSeconds))
    (Rendering|Material|SetScalarParameterValueonMaterials _fieldfx "HitWave-Duration" 0.1)
    (Components|Activation|Activate (Variables|Default|GetNS_Grumpy_Beam) true)
    (Development|PrintString "ACTIVATE - RESET" true true "(R=1.000000,G=0.000000,B=0.790744,A=1.000000)")))

(event Custom|EdgeServerApplyDamage
  (params (TotalDamage :type Integer64))
  (CallFunction|SetBigHealth TotalDamage true)
  (Health|SetHealth (Math|Float|Min(Float) (Math|Conversions|ToFloat(Integer64) TotalDamage) (Health|GetMaxHealth)))
  (Networking|ForceNetUpdate))

(event Custom|ClientSwapColor
  (params (Color :type Byte))
  (Variables|Default|SetColor Color)
  (CallFunction|UpdateClientVisuals))

(event Custom|UpdateClientVisuals
  (bind _color (Variables|Default|GetColor))
  (bind (_relativelocation_x _relativelocation_y) (Math|Vector|BreakVector (Variables|/Script/Engine.SceneComponent|GetRelativeLocation_X (Variables|Default|GetSphere))))
  (bind _ns_grumpy_beam (Variables|Default|GetNS_Grumpy_Beam))
  ;[1 1 1] Visuals
  (if (Math|Byte|Equal(Byte) (Variables|Default|GetRepCompleteState) 0)
    (Utilities|FlowControl|Sequence
      (:then_0
        (Rendering|Material|SetScalarParameterValueonMaterials (Variables|Default|GetColorPlane) "IsCompleted" 0.0)
        (Rendering|Material|SetScalarParameterValueonMaterials (Variables|Default|GetColorPlane) "Fade" 1.0)
        (CallFunction|SetupTimelineColorParams (Variables|Default|GetColorBlocks))
        (CallFunction|SetupTimelineColorParams (Variables|Default|GetColorDots))
        (CallFunction|SetupTimelineColorParams (Variables|Default|GetBaseColor))
        (Effects|Components|ParticleSystem|SetVectorParameter (Variables|Default|GetNS_Grumpy_Beam) "CentralBeamColor" (Utilities|Select :Index _color :Option%200 (Math|Conversions|ToVector(LinearColor) (Variables|Default|GetCentralBeamColor_Red)) :Option%201 (Math|Conversions|ToVector(LinearColor) (Variables|Default|GetCentralBeamColor_Green)) :Option%202 (Math|Conversions|ToVector(LinearColor) (Variables|Default|GetCentralBeamColor_Blue))))
        (Rendering|Material|SetVectorParameterValueonMaterials (Variables|Default|GetColorPlane) "EdgeColor" (Utilities|Select :Index _color :Option%200 (Math|Conversions|ToVector(LinearColor) (Variables|Default|GetTransition_Red)) :Option%201 (Math|Conversions|ToVector(LinearColor) (Variables|Default|GetTransition_Green)) :Option%202 (Math|Conversions|ToVector(LinearColor) (Variables|Default|GetTransition_Blue))))
        (Components|Timeline|PlayfromStart (Variables|BP_Grumpy_Cube_Child|GetTransitionColorsTL))
        (Audio|PlaySoundatLocation "/Figment_S08/Athena/Sounds/Events/Dopey/Grumpy/Grumpy_Panel_Change_Cue.Grumpy_Panel_Change_Cue" (Transformation|GetWorldLocation (Variables|Default|GetAudioAmbient))))
      (:then_1
        (Transformation|SetRelativeLocation (Variables|Default|GetSphere) (Math|Vector|MakeVector _relativelocation_x _relativelocation_y 1.6e+02) false false)
        (Development|SetHiddeninGame (Variables|Default|GetFieldFX) false false)
        (CallFunction|SetSuppressHealthBar self false)
        (Rendering|Material|SetScalarParameterValueonMaterials (Variables|Default|GetColorPlane) "Fade" 1.0)
        (Effects|Components|ParticleSystem|SetFloatParameter _ns_grumpy_beam "Alpha" 1.0)
        (Effects|Components|ParticleSystem|SetVectorParameter _ns_grumpy_beam "FadeColor" "1.000000,1.000000,1.000000")))))

(event Custom|PlaySuccessVisuals
  (bind _self self)
  (bind _colorplane (Variables|Default|GetColorPlane))
  (bind (_relativelocation_x _relativelocation_y) (Math|Vector|BreakVector (Variables|/Script/Engine.SceneComponent|GetRelativeLocation_X (Variables|Default|GetSphere))))
  (bind _ns_grumpy_beam (Variables|Default|GetNS_Grumpy_Beam))
  (Utilities|FlowControl|Delay (Variables|Default|GetDelayBeforeDeactivation))
  (Networking|FlushNetDormancy)
  (bind (_direction _fade) (Timeline|DeactivateFloor PlayFromStart)
    (:Update
      (Rendering|Material|SetScalarParameterValueonMaterials (Variables|Default|GetColorPlane) "Fade" _fade)
      (Effects|Components|ParticleSystem|SetFloatParameter _ns_grumpy_beam "Alpha" _fade)
      (Effects|Components|ParticleSystem|SetVectorParameter _ns_grumpy_beam "FadeColor" (Math|Conversions|ToVector(Float) _fade)))
    (:Finished
      (Utilities|FlowControl|Sequence
        (:then_0
          (Audio|Components|Audio|FadeOut (Variables|Default|GetAudioAmbient) 2.0 0.0 "Linear" :!class /Script/Engine.AudioComponent)
          (bind _gridambient (Variables|Default|GetGridAmbient)
            (:then
              (Audio|Components|Audio|FadeOut _gridambient 2.0 0.0 "Linear" :!class /Script/Engine.AudioComponent))
            (:else)))
        (:then_1
          (Utilities|FlowControl|Delay (Math|Random|RandomFloatinRange (Variables|Default|GetDelayBeforeEyeFallsMin) (Variables|Default|GetDelayBeforeEyeFallsMax)))
          (Networking|FlushNetDormancy)
          (bind (_direction_2 _relativelocationz) (Timeline|EyeFallTL PlayFromStart)
            (:Update
              (Transformation|SetRelativeLocation (Variables|Default|GetSphere) (Math|Vector|MakeVector _relativelocation_x _relativelocation_y _relativelocationz) false false))
            (:Finished
              (Networking|SwitchHasAuthority
                (:Authority
                  (CallFunction|UpdateCompleteState 1))))
            (:SpawnVFX
              (Effects|SpawnEmitterAttached "/Game/Athena/Items/Consumables/C4/P_Athena_C4_Land_Dust.P_Athena_C4_Land_Dust" (Variables|Transformation|GetRootComponent) "None" "0,256.000000,18.000000" "90.000000, 0, 0" "2.000000,2.000000,2.000000"))))
        (:then_2
          _))))
  (Utilities|FlowControl|DoOnce false
    (:Completed
      (Fortnite|SetCanbeDamaged _self)
      (CallFunction|SetSuppressHealthBar _self true)
      (Development|SetHiddeninGame (Variables|Default|GetFieldFX) true false)
      (Rendering|Material|SetScalarParameterValueonMaterials _colorplane "IsCompleted" 1.0)
      (Rendering|Material|SetScalarParameterValueonMaterials _colorplane "Fade" 1.0))))

(event Custom|SetCompleteState
  (params (Runtime :type Boolean))
  (bind _self self)
  (bind (_relativelocation_x _relativelocation_y) (Math|Vector|BreakVector (Variables|/Script/Engine.SceneComponent|GetRelativeLocation_X (Variables|Default|GetSphere))))
  (Utilities|FlowControl|Sequence
    (:then_0
      (Development|SetHiddeninGame (Variables|Default|GetFieldFX) true false)
      (Rendering|Material|SetScalarParameterValueonMaterials (Variables|Default|GetColorPlane) "Fade" 0.0)
      (Rendering|Material|SetScalarParameterValueonMaterials (Variables|Default|GetColorPlane) "IsCompleted" 1.0)
      (Fortnite|SetCanbeDamaged _self)
      (CallFunction|SetSuppressHealthBar _self true)
      (Components|Activation|Deactivate (Variables|Default|GetAudioAmbient)))
    (:then_2
      (if Runtime
        (Transformation|SetRelativeLocation (Variables|Default|GetSphere) (Math|Vector|MakeVector _relativelocation_x _relativelocation_y 4e+01) false false)
        (else
          (Development|SetHiddeninGame (Variables|Default|GetHitPoint) true false))))))

(event Custom|PlaySuccessSweepFX)

(fn OnRep_RepCompleteState
  (Networking|SwitchHasAuthority
    (:Authority
      (switch int (Math|Conversions|ToInteger(Byte) (Variables|Default|GetRepCompleteState))
        (:2
          (Collision|SetCollisionEnabled (Variables|Default|GetSphere) "NoCollision"))))
    (:Remote
      (switch int (Math|Conversions|ToInteger(Byte) (Variables|Default|GetRepCompleteState))
        (:0
          (CallFunction|ActiveSetup))
        (:1
          (CallFunction|SetCompleteState true))
        (:2
          (CallFunction|SetCompleteState false)
          (Collision|SetCollisionEnabled (Variables|Default|GetSphere) "NoCollision"))))))

(fn SetupTimelineColorParams
  (params (ColorMap :type Map%20of%20Linear%20Color%20Structures%20to%20Names))
  (bind _color (Variables|Default|GetColor))
  (bind _returnvalue (Math|Conversions|ToInteger(Byte) _color))
  (bind _returnvalue_2 (- _returnvalue 1))
  (bind _values (Utilities|Map|Values (Variables|GetColorMap)))
  (bind _keys (Utilities|Map|Keys (Variables|GetColorMap)))
  (Rendering|Material|SetVectorParameterValueonMaterials (Variables|Default|GetColorPlane) (Utilities|String|StringToName (Utilities|String|Append (Utilities|String|ToString(Name) (Utilities|Array|Get(acopy) _values (Math|Integer|%25%28Integer%29 (+ _returnvalue_2 (Utilities|Array|Length _values)) (Utilities|Array|Length _values)))) "Start")) (Math|Conversions|ToVector(LinearColor) (Utilities|Array|Get(acopy) _keys (Math|Integer|%25%28Integer%29 (+ _returnvalue_2 (Utilities|Array|Length _values)) (Utilities|Array|Length _values)))))
  (Rendering|Material|SetVectorParameterValueonMaterials (Variables|Default|GetColorPlane) (Utilities|Array|Get(acopy) _values (Math|Integer|%25%28Integer%29 (+ _returnvalue_2 (Utilities|Array|Length _values)) (Utilities|Array|Length _values))) (Math|Conversions|ToVector(LinearColor) (Utilities|Array|Get(acopy) _keys _returnvalue))))

(fn ShouldDie
  (params (Damage :type Float%20%28single-precision%29) (EventInstigator :type Controller%20Object%20Reference) (DamageCauser :type Actor%20Object%20Reference) (bIsTest :type Boolean) (EffectContext :type Gameplay%20Effect%20Context%20Handle%20Structure%20%28by%20ref%29))
  (return false))

(fn UpdateCompleteState
  (params (State :type Byte))
  (if (or (>= State (Variables|Default|GetRepCompleteState)) (== (Variables|Default|GetRepCompleteState) 255))
    (Variables|Default|SetRepCompleteState State)))

(fn ConstructionScript)
