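# Default Lit base material for the Grumpy game prototype, drawing an animated screen-like surface of tiled, offset RGB dots over blocks and a bump-offset grid.
# A noise-distorted radial Transition mask blends start colors into final colors, with a glowing edge.
# Randomized per-block sine pulses drive dot and block brightness, switching to a synchronized flash through IsCompleted, while Fade scales the results.
# It also contains blended detail normals and a panning-noise circle color layer.
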Material /Figment_S08/Athena/Prototype/Blueprints/Grumpy/Grumpy_Game_BaseMaterial
graph {
 properties {bUsedWithInstancedStaticMeshes=true, bUsedWithStaticLighting=true, pixelDepthOffsetMode=PDOM_Legacy, usedShadingModels=Default Lit}
 nodes [
  ConstantBiasScale(None=Sine(None=Time()) {period=6})
  Multiply(
   A=MaterialFunctionCall(
    BaseNormal=TextureSample() {samplerType=SAMPLERTYPE_Normal, texture=/Game/Global/Textures/Default/Noise/T_SwirlNoise_N}.RGB
    AdditionalNormal=TextureSample() {samplerType=SAMPLERTYPE_Normal, texture=/Game/Environments/Sets/Arid/Textures/RockDetail_normal}.RGB
   ) {materialFunction=&BlendAngleCorrectedNormals}.Result
   B=VectorParameter() {parameterName=NormalStrength, defaultValue=<rgba 0.02 0.02 1 0>}.RGB
  )
  n1 = ScalarParameter() {defaultValue=1, parameterName=Fade}
  n2 = Frac(None=Multiply(A=TextureCoordinate(), B=ScalarParameter() {defaultValue=50, parameterName=DotsRepeat}))
  n3 = Add(A=@n2, B=AppendVector(A=ScalarParameter() {defaultValue=0.11, parameterName=R_Offset_U}, B=ScalarParameter() {parameterName=R_Offset_V}))
  n4 = Constant() {r=0.5}
  n5 = ScalarParameter() {defaultValue=0.6, parameterName=Radius}
  n6 = ScalarParameter() {defaultValue=0.2, parameterName=DotsHardness}
  n7 = ComponentMask(None=TextureSample(UVs=TextureCoordinate() {uTiling=2, vTiling=2}) {samplerType=SAMPLERTYPE_Normal, texture=&T_SwirlNoise_N}.RGB) {g=true, r=true}
  n8 = ScalarParameter() {defaultValue=0.06, parameterName=TransitionNoiseScale}
  n9 = SphereMask(
   A=Add(A=Multiply(A=@n7, B=@n8), B=TextureCoordinate())
   B=Constant() {r=0.5}
   Radius=ScalarParameter() {parameterName=Transition}
   Hardness=ScalarParameter() {defaultValue=0.6, parameterName=TransitionHardness}
  )
  n10 = MaterialFunctionCall(Float3=LinearInterpolate(
    A=VectorParameter() {parameterName=ColorDotsStart, defaultValue=<rgba 0 0 0 1>}.RGB
    B=VectorParameter() {parameterName=ColorDots, defaultValue=<rgba 0 0 0 1>}.RGB
    Alpha=@n9
   )) {materialFunction=&BreakOutFloat3Components}
  n11 = Add(A=@n2, B=AppendVector(A=ScalarParameter() {parameterName=G_Offset_U}, B=ScalarParameter() {defaultValue=-0.1, parameterName=G_Offset_V}))
  n12 = Constant() {r=0.5}
  n13 = ScalarParameter() {defaultValue=0.6, parameterName=Radius}
  n14 = ScalarParameter() {defaultValue=0.2, parameterName=DotsHardness}
  n15 = Add(A=@n2, B=AppendVector(A=ScalarParameter() {defaultValue=-0.11, parameterName=B_Offset_U}, B=ScalarParameter() {parameterName=B_Offset_V}))
  n16 = Constant() {r=0.5}
  n17 = ScalarParameter() {defaultValue=0.6, parameterName=Radius}
  n18 = ScalarParameter() {defaultValue=0.2, parameterName=DotsHardness}
  n19 = MaterialFunctionCall(
   X=Multiply(A=SphereMask(A=@n3, B=@n4, Radius=@n5, Hardness=@n6), B=@n10.R)
   Y=Multiply(A=SphereMask(A=@n11, B=@n12, Radius=@n13, Hardness=@n14), B=@n10.G)
   Z=Multiply(A=@n10.B, B=SphereMask(A=@n15, B=@n16, Radius=@n17, Hardness=@n18))
  ) {materialFunction=&MakeFloat3}
  n20 = TextureCoordinate()
  n21 = ScalarParameter() {defaultValue=12.5, parameterName=BlockTile}
  n22 = Add(
   A=MaterialFunctionCall!(
    PrimaryTextureCoordinates=Multiply(A=@n20, B=@n21)
    2ndOrderTextureCoordinates
    RandomSeed
    NoiseDensityBias
    Bypass Noise Influence Crosstalk
   ) {materialFunction=/Game/Effects/Fort_Effects/Materials/_Functions/MF_RandomUVSample}.NoiseWithRange
   B=Multiply(A=Time(), B=ScalarParameter() {defaultValue=0.6, parameterName=BlockSpeed})
  )
  n23 = Sine(None=Time() {bOverride_Period=true, period=1}) {period=0.5}
  n24 = LinearInterpolate(
   A=Add(A=Sine(None=@n22))
   B=Multiply(A=Add(A=@n23) {constB=1.5}, B=ScalarParameter() {defaultValue=40, parameterName=CompletedFlashScale})
   Alpha=ScalarParameter() {parameterName=IsCompleted}
  )
  n25 = Saturate(None=@n24)
  n26 = Multiply(
   A=SphereMask(A=@n2, B=Constant() {r=0.5}, Radius=ScalarParameter() {defaultValue=0.4, parameterName=RadiusMask}, Hardness=Constant() {r=0.8})
   B=Power(Base=Multiply(A=@n19.Result, B=@n25), Exp=ScalarParameter() {defaultValue=5, parameterName=DotsExponent})
  )
  n27 = ScalarParameter() {defaultValue=0.01, parameterName=BlocksScale}
  n28 = TextureCoordinate()
  n29 = ScalarParameter() {defaultValue=20, parameterName=GridTile}
  n30 = TextureSample(UVs=BumpOffset(Coordinate=Multiply(A=@n28, B=@n29), Height=ScalarParameter() {defaultValue=-4, parameterName=GridBumpOffsetHeight})) {texture=/Game/Weapons/Prototype/ObjectMover/Textures/T_CreativeMoveTool_Grid}
  n31 = ScalarParameter() {defaultValue=0.15, parameterName=GridScale}
  n32 = Multiply(
   A=Add(A=Multiply(A=@n24, B=@n27), B=Multiply(A=@n30.R, B=@n31))
   B=LinearInterpolate(
    A=VectorParameter() {parameterName=ColorBlocksStart, defaultValue=<rgba 0 0 0 1>}.RGB
    B=VectorParameter() {parameterName=ColorBlocks, defaultValue=<rgba 0 0 0 1>}.RGB
    Alpha=@n9
   )
  )
  n33 = Multiply(
   A=SphereMask(A=@n9, B=Constant() {r=0.5}, Radius=ScalarParameter() {defaultValue=0.1, parameterName=EdgeRadius}, Hardness=Constant())
   B=VectorParameter() {parameterName=EdgeColor, defaultValue=<rgba 0.1971 0.3955 1 0>}.RGB
  )
  n34 = ScalarParameter() {defaultValue=6, parameterName=EdgePower}
  Multiply(A=@n1, B=Add(A=Add(A=@n26, B=@n32), B=Power(Base=@n33, Exp=@n34)))
  n35 = Multiply(
   A=SphereMask(A=TextureCoordinate(), B=Constant() {r=0.5}, Radius=ScalarParameter() {defaultValue=0.4, parameterName=BGRadius}, Hardness=Constant())
   B=ScalarParameter() {defaultValue=1.5, parameterName=BGScale}
  )
  n36 = LinearInterpolate(
   A=VectorParameter() {parameterName=BaseColorStart, defaultValue=<rgba 0 0 0 1>}.RGB
   B=VectorParameter() {parameterName=BaseColor, defaultValue=<rgba 0 0 0 1>}.RGB
   Alpha=@n9
  )
  n37 = Multiply(A=@n36) {constB=0.72}
  n38 = DotProduct(A=@n19.Result, B=Constant() {r=1})
  Multiply(A=Add(A=Multiply(A=@n35, B=@n36), B=LinearInterpolate(A=@n36, B=@n37, Alpha=@n38) {constB=0}), B=@n1)
  n39 = VectorParameter() {parameterName=PanSpeedNoiseA, defaultValue=<rgba 7.5 15 11 0>}
  n40 = TextureSampleParameter2D(UVs=Divide(A=Panner(Speed=@n39.RGB) {speedX=5}, B=ScalarParameter() {defaultValue=1000, parameterName=RepeatNoiseA})) {parameterName=Noise, texture=/Figment_S08/Athena/Environments/DudeBro/T_Dudebro_Noise_01}
  n41 = VectorParameter() {parameterName=PanSpeedNoiseB, defaultValue=<rgba -11 -32.02 0 0>}
  n42 = TextureSampleParameter2D(UVs=Divide(A=Panner(Speed=@n41.RGB) {speedX=5}, B=ScalarParameter() {defaultValue=-1300, parameterName=RepeatNoiseB})) {parameterName=Noise, texture=&T_Dudebro_Noise_01}
  n43 = Time()
  n44 = ScalarParameter() {defaultValue=0.35, parameterName=NoiseSineSpeed}
  n45 = MaterialFunctionCall(Float2=TextureCoordinate()) {materialFunction=&BreakOutFloat2Components}
  n46 = Sine(None=Add(A=Multiply(A=@n43, B=@n44), B=@n45.G)) {period=3}
  n47 = Multiply(
   A=SphereMask(
    A=Multiply(A=@n40.R, B=@n42.R)
    B=ConstantBiasScale(None=@n46)
    Radius=ScalarParameter() {defaultValue=0.5, parameterName=NoiseRadius}
    Hardness=ScalarParameter() {parameterName=NoiseHardness}
   )
   B=ScalarParameter() {defaultValue=0.05, parameterName=NoiseScale}
  )
  n48 = Constant() {r=0.5}
  Multiply(
   A=LinearInterpolate(
    A=VectorParameter() {parameterName=BGColor, defaultValue=<rgba 0.7448 0.0124 0 0>}.RGB
    B=VectorParameter() {parameterName=Circles, defaultValue=<rgba 10 1.6189 0.3215 0>}.RGB
    Alpha=Add(B=@n47)
   )
   B=SphereMask(A=@n45.G, B=@n48, Radius=@n48, Hardness=Constant())
  )
  Multiply(
   A=ScalarParameter() {defaultValue=0.65, parameterName=Metallic}
   B=TextureSample() {samplerType=SAMPLERTYPE_Masks, texture=/Game/Environments/Sets/Arid/Textures/T_desert_sand_base_S}.B
  )
  ScalarParameter() {defaultValue=0.22, parameterName=Roughness}
  ScalarParameter() {defaultValue=0.2, parameterName=Specular}
 ]
}
